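# Sound cue for the Season 36 MistRaven story quest shard sign, apparently heard when the sign is placed and as it lingers.
# A mixer layers a one-shot Place sound with a looping wave that starts after a short fixed delay.
# It borrows the hidden-star ambient loop attenuation, uses the non-critical in-game UI sound class, and overrides concurrency to one instance per owner, blocking new plays.
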SoundCue /BattlePassS36_Quests/Audio/SFX/Quest_S36_Story_MistRaven_ShardSign/Quest_S36_Story_MistRaven_ShardSign_Cue
root=Mixer_0
nodes [
 {id=Delay_0, children [WavePlayer_1], properties {delayMax=0.2, delayMin=0.2}}
 {id=Mixer_0, children [WavePlayer_0,Delay_0], properties {inputVolume [1,1]}}
 {
  id=WavePlayer_0
  properties {soundWaveAssetPtr=&/Quest_S36_Story_MistRaven_ShardSign_Place}
 }
 {
  id=WavePlayer_1
  properties {
   bLooping=true
   soundWaveAssetPtr=&/Quest_S36_Story_MistRaven_ShardSign_Loop
  }
 }
]
properties {
 attenuationSettings=/Game/Athena/Sounds/SupplyDrops/Fort_HiddenStar_Ambient_Loop_Att
 bOverrideConcurrency=true
 maxDistance=2500
 soundClassObject=&Fort_UI_InGame_NonCritical
 volumeMultiplier=2
 concurrencyOverrides {bLimitToOwner=true, maxCount=1, resolutionRule=PreventNew}
}
